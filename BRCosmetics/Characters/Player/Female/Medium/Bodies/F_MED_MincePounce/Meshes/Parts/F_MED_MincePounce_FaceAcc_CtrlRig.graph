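# Control Rig for the MincePounce female medium body's face accessory mesh that repositions the ponytail according to head pose.
# Three chained spherical pose readers on the head bone, two measured relative to spine_05, each weight an offset on MSTR_ponytail that propagates down the chain.
# The first offset rotates and shifts the ponytail, while the other two rotate it in opposite directions.
# Bang and earphone bones are not driven, and the third reader has debug drawing enabled.
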ControlRigBlueprint /BRCosmetics/Characters/Player/Female/Medium/Bodies/F_MED_MincePounce/Meshes/Parts/F_MED_MincePounce_FaceAcc_CtrlRig
graphs [
 RigVMFunctionLibrary {links [], nodes []}
 RigVMModel {nodes [
   RigUnit_BeginExecution = &RigUnit_BeginExecution()
   n1 = &RigUnit_SphericalPoseReader(
    ActiveRegionScaleFactors=(PositiveWidth=1.5,NegativeWidth=1.5,PositiveHeight=1,NegativeHeight=1)
    ActiveRegionSize=0.13
    Debug=(bDrawDebug=false,bDraw2D=False,bDrawLocalAxes=False,DebugScale=15,DebugSegments=20,DebugThickness=0.1)
    DriverAxis=(X=1,Y=0,Z=0)
    DriverItem=(Type=Bone,Name="head")
    FalloffRegionScaleFactors=(PositiveWidth=1,NegativeWidth=1,PositiveHeight=1,NegativeHeight=1)
    FalloffSize=0.3
    FlipHeightScaling=False
    FlipWidthScaling=False
    OptionalParentItem=(Type=Bone,Name="None")
    RotationOffset=(X=-45,Y=-90,Z=0)
    ExecutePin=@RigUnit_BeginExecution.ExecutePin
   )
   > RigUnit_OffsetTransformForItem = &RigUnit_OffsetTransformForItem(
    Item=(Type=Bone,Name="MSTR_ponytail")
    OffsetTransform=(Rotation=(X=0,Y=-0.4226,Z=0,W=0.9063),Translation=(X=0,Y=0,Z=3),Scale3D=(X=1,Y=1,Z=1))
    bPropagateToChildren=True
    Weight=@n1.OutputParam
   )
   > n2 = &RigUnit_SphericalPoseReader(
    ActiveRegionScaleFactors=(PositiveWidth=1,NegativeWidth=1,PositiveHeight=1,NegativeHeight=1)
    ActiveRegionSize=0.06
    Debug=(bDrawDebug=false,bDraw2D=False,bDrawLocalAxes=False,DebugScale=15,DebugSegments=20,DebugThickness=0.1)
    DriverAxis=(X=0,Y=1,Z=0)
    DriverItem=(Type=Bone,Name="head")
    FalloffRegionScaleFactors=(PositiveWidth=1,NegativeWidth=1,PositiveHeight=1,NegativeHeight=1)
    FalloffSize=0.5
    FlipHeightScaling=False
    FlipWidthScaling=False
    OptionalParentItem=(Type=Bone,Name="spine_05")
    RotationOffset=(X=0,Y=0,Z=0)
   )
   > &RigUnit_OffsetTransformForItem(
    Item=(Type=Bone,Name="MSTR_ponytail")
    OffsetTransform=(Rotation=(X=0,Y=0,Z=-0.5,W=0.866),Translation=(X=0,Y=0,Z=0),Scale3D=(X=1,Y=1,Z=1))
    bPropagateToChildren=True
    Weight=@n2.OutputParam
   )
   > n3 = &RigUnit_SphericalPoseReader(
    ActiveRegionScaleFactors=(PositiveWidth=2,NegativeWidth=3,PositiveHeight=1,NegativeHeight=1)
    ActiveRegionSize=0.09
    Debug=(bDrawDebug=true,bDraw2D=False,bDrawLocalAxes=False,DebugScale=15,DebugSegments=20,DebugThickness=0.1)
    DriverAxis=(X=0,Y=1,Z=0)
    DriverItem=(Type=Bone,Name="head")
    FalloffRegionScaleFactors=(PositiveWidth=1,NegativeWidth=1,PositiveHeight=1,NegativeHeight=1)
    FalloffSize=0.5
    FlipHeightScaling=False
    FlipWidthScaling=False
    OptionalParentItem=(Type=Bone,Name="spine_05")
    RotationOffset=(X=180,Y=0,Z=0)
   )
   > &RigUnit_OffsetTransformForItem(
    Item=(Type=Bone,Name="MSTR_ponytail")
    OffsetTransform=(Rotation=(X=0,Y=0,Z=0.2164,W=0.9763),Translation=(X=0,Y=0,Z=0),Scale3D=(X=1,Y=1,Z=1))
    bPropagateToChildren=True
    Weight=@n3.OutputParam
   )
  ]}
]
hierarchy [
 root {loc=<0 0 0>, rot=<0 0 0 1>}
 pelvis {parent=root, loc=<0 -0.5301 98.0571>, rot=<0.0127 0.707 -0.0127 -0.707>}
 spine_01 {parent=pelvis, loc=<6.1208 -0.9328 0>, rot=<0 0 0.034 -0.9994>}
 spine_02 {parent=spine_01, loc=<6.5997 -0.2095 0>, rot=<0 0 0.0238 -0.9997>}
 spine_03 {parent=spine_02, loc=<6.4914 -0.1237 0>, rot=<0 0 0.0149 -0.9999>}
 spine_04 {parent=spine_03, loc=<6.9598 0.0794 0>, rot=<0 0 -0.0132 -0.9999>}
 spine_05 {parent=spine_04, loc=<6.9987 0 0>, rot=<0 0 -0.0015 -1>}
 neck_01 {parent=spine_05, loc=<11.3138 0.1529 0>, rot=<0 0 -0.2003 -0.9797>}
 neck_02 {parent=neck_01, loc=<4.5627 0 0>, rot=<0 0 0.0504 -0.9987>}
 head {parent=neck_02, loc=<5.3914 0 0>, rot=<0 0 0.115 -0.9934>}
 MSTR_ponytail {parent=head, loc=<2.1368 -4.2583 0>, rot=<0.1631 -0.688 0.688 -0.1631>}
 ponytail_1 {parent=MSTR_ponytail, loc=<0 0 0>, rot=<0 0 0 1>}
 ponytail_2 {parent=ponytail_1, loc=<4.5851 0 0>, rot=<0 0.0534 0 -0.9986>}
 ponytail_3 {parent=ponytail_2, loc=<5.9069 0 0>, rot=<0 0.0794 0 -0.9968>}
 ponytail_4 {parent=ponytail_3, loc=<6.8922 0 0>, rot=<0 -0.0794 0 -0.9968>}
 ponytail_5 {parent=ponytail_4, loc=<6.9087 0 0>, rot=<0 -0.2538 0 -0.9673>}
 ponytail_6 {parent=ponytail_5, loc=<7.0295 0 0>, rot=<0 -0.1079 0 -0.9942>}
 ponytail_7 {parent=ponytail_6, loc=<7.3118 0 0>, rot=<0 0.1219 0 -0.9925>}
 ponytail_8 {parent=ponytail_7, loc=<6.2156 0 0>, rot=<0 0.1535 0 -0.9882>}
 ponytail_deform_base {parent=ponytail_1, loc=<0.2725 0 0.434>, rot=<0 0.4839 0 -0.8752>}
 ponytail_deform {parent=ponytail_deform_base, loc=<0 0 0>, rot=<0 0 0 1>}
 MSTR_bang_r_1 {parent=head, loc=<6.7323 4.4659 6.7957>, rot=<0.0736 0.1322 0.9878 -0.0374>}
 bang_r_1 {parent=MSTR_bang_r_1, loc=<0 0 0>, rot=<0 0 0 1>}
 bang_r_2 {parent=bang_r_1, loc=<4.3154 0 0>, rot=<-0.0898 -0.1769 0.0632 -0.9781>}
 bang_r_3 {parent=bang_r_2, loc=<4.7261 0 0>, rot=<0 0 0.1626 -0.9867>}
 MSTR_bang_l_1 {parent=head, loc=<6.7321 4.4659 -6.7957>, rot=<0.0374 -0.9878 -0.1322 -0.0736>}
 bang_l_1 {parent=MSTR_bang_l_1, loc=<0 0 0>, rot=<0 0 0 -1>}
 bang_l_2 {parent=bang_l_1, loc=<4.3151 0 0>, rot=<0.0898 -0.1769 -0.0632 -0.9781>}
 bang_l_3 {parent=bang_l_2, loc=<4.726 0 0>, rot=<0 0 -0.1626 -0.9867>}
 earphone_r_deform_base {parent=head, loc=<3.2502 2.0331 7.6805>, rot=<0.661 0.1228 0.7388 0.0459>}
 earphone_r_deform {parent=earphone_r_deform_base, loc=<0 0 0>, rot=<0 0 0 1>}
 earphone_l_deform_base {parent=head, loc=<3.2501 2.0331 -7.6804>, rot=<-0.0459 -0.7388 -0.1228 -0.661>}
 earphone_l_deform {parent=earphone_l_deform_base, loc=<0 0 0>, rot=<0 0 0 1>}
 clavicle_l {parent=spine_05, loc=<5.4831 -0.1053 -3.2946>, rot=<0.0465 -0.7289 -0.009 -0.6829>}
 upperarm_l {parent=clavicle_l, loc=<9.8658 -1.2136 -0.4105>, rot=<0.0164 -0.3245 -0.0269 -0.9454>}
 upperarm_twist_01_l {parent=upperarm_l, loc=<7.2135 -0.1134 0>, rot=<0 0 0 1>}
 clavicle_r {parent=spine_05, loc=<5.4831 -0.1053 3.2946>, rot=<0.7289 0.0465 -0.6829 0.009>}
 upperarm_r {parent=clavicle_r, loc=<-9.8658 1.2136 0.4105>, rot=<0.0164 -0.3245 -0.0269 -0.9454>}
 upperarm_twist_01_r {parent=upperarm_r, loc=<-7.2135 0.1134 0>, rot=<0 0 0 1>}
]
