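# Physics asset for the ponytail on the MincePounce female medium character's face accessory mesh, providing bodies for simulated hair motion.
# Eight ponytail capsules, fattest in the middle of the chain, link back to a small MSTR_ponytail root through constraints with angular swing and tight twist limits.
# Unconstrained capsules on the spine, neck, clavicles and upper arms supply body collision shapes, with spine_05 carrying a second, large offset capsule.
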PhysicsAsset /BRCosmetics/Characters/Player/Female/Medium/Bodies/F_MED_MincePounce/Meshes/Parts/F_MED_MincePounce_FaceAcc_Physics_Ponytail
bodies {
 MSTR_ponytail [{shapeType=Capsule, length=1, radius=1, shapeName=MSTR_ponytail_capsule, rotation=<rot 0 -90 0>}]
 clavicle_l [{
   shapeType=Capsule
   length=8.3414
   radius=6.7987
   shapeName=clavicle_l_capsule
   center=<7.8834 -0.8055 -0.5183>
   rotation=<rot 83.9057 5.6844 7.1218>
  }]
 clavicle_r [{
   shapeType=Capsule
   length=5.4476
   radius=6.4868
   shapeName=clavicle_r_capsule
   center=<-8.0769 1.2739 0.8251>
   rotation=<rot 77.4278 -0.0002 -0.0002>
  }]
 neck_01 [{shapeType=Capsule, length=0.0475, radius=6.6078, shapeName=neck_01_capsule, center=<1.68 1.4326 0>, rotation=<rot 90 0 0>}]
 ponytail_1 [{shapeType=Capsule, length=3, radius=2, shapeName=ponytail_1_capsule, center=<2.8632 0 0>, rotation=<rot 90 0 0>}]
 ponytail_2 [{shapeType=Capsule, length=3, radius=2, shapeName=ponytail_2_capsule, center=<2.9417 0 0>, rotation=<rot 90 0 0>}]
 ponytail_3 [{shapeType=Capsule, length=3.284, radius=1.716, shapeName=ponytail_3_capsule, center=<3.1451 0 0>, rotation=<rot 90 0 0>}]
 ponytail_4 [{shapeType=Capsule, length=0.2189, radius=4.6754, shapeName=ponytail_4_capsule, center=<3.2241 0 1.6376>, rotation=<rot 90 0 0>}]
 ponytail_5 [{shapeType=Capsule, length=0.4796, radius=5.8204, shapeName=ponytail_5_capsule, center=<3.0942 0.3602 1.9433>, rotation=<rot 90 0 0>}]
 ponytail_6 [{shapeType=Capsule, length=1.1852, radius=3.8148, shapeName=ponytail_6_capsule, center=<3.2561 0 0.7907>, rotation=<rot 90 0 0>}]
 ponytail_7 [{shapeType=Capsule, length=3.1109, radius=1.8891, shapeName=ponytail_7_capsule, center=<2.8632 0 0>, rotation=<rot 90 0 0>}]
 ponytail_8 [{shapeType=Capsule, length=3, radius=2, shapeName=ponytail_8_capsule, center=<2.8632 0 0>, rotation=<rot 90 0 0>}]
 spine_01 [{
   shapeType=Capsule
   length=7.3891
   radius=13.6092
   shapeName=spine_01_capsule
   center=<-14.4886 -9.0494 -0.0999>
   rotation=<rot 90 -89.4722 0>
  }]
 spine_03 [{shapeType=Capsule, length=5.608, radius=10.3075, shapeName=spine_03_capsule, center=<-7.9334 1.1692 0>, rotation=<rot 0 -79.8728 90>}]
 spine_04 [{shapeType=Capsule, length=8.1307, radius=9.2414, shapeName=spine_04_capsule}]
 spine_05 [
  {shapeType=Capsule, length=5.1833, radius=9.7645, shapeName=spine_05_capsule, center=<3.6573 -0.2952 0>, rotation=<rot 0 -90 0>}
  {shapeType=Capsule, length=24.3799, radius=15, shapeName=spine_05_capsule, center=<13.7486 16.188 0>}
 ]
 upperarm_l [{shapeType=Capsule, length=12.3506, radius=4.4205, shapeName=upperarm_l_capsule, center=<13.9756 0 -0.4444>, rotation=<rot 90 0 0>}]
 upperarm_r [{shapeType=Capsule, length=12.3506, radius=4.4205, shapeName=upperarm_r_capsule, center=<-13.7581 0 0.5041>, rotation=<rot 90 0 0>}]
}
constraints [
 {bone1Name=ponytail_8, bone2Name=ponytail_7, swing1LimitDegrees=15, swing2LimitDegrees=30, twistLimitDegrees=5}
 {bone1Name=ponytail_7, bone2Name=ponytail_6, swing1LimitDegrees=15, swing2LimitDegrees=30, twistLimitDegrees=5}
 {bone1Name=ponytail_6, bone2Name=ponytail_5, swing1LimitDegrees=15, swing2LimitDegrees=30, twistLimitDegrees=5}
 {bone1Name=ponytail_5, bone2Name=ponytail_4, swing1LimitDegrees=15, swing2LimitDegrees=30, twistLimitDegrees=5}
 {bone1Name=ponytail_4, bone2Name=ponytail_3, swing1LimitDegrees=15, swing2LimitDegrees=20, twistLimitDegrees=5}
 {bone1Name=ponytail_3, bone2Name=ponytail_2, swing1LimitDegrees=15, swing2LimitDegrees=30, twistLimitDegrees=5}
 {bone1Name=ponytail_2, bone2Name=ponytail_1, swing1LimitDegrees=15, swing2LimitDegrees=30, twistLimitDegrees=5}
 {bone1Name=ponytail_1, bone2Name=MSTR_ponytail, swing1LimitDegrees=15, swing2LimitDegrees=20, twistLimitDegrees=5}
]
