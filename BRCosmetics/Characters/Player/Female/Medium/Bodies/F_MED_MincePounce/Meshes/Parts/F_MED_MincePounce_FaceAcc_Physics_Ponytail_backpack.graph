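# Physics asset for the MincePounce female character cosmetic's face accessory ponytail, in a variant for wearing a backpack.
# Capsule bodies on ponytail_5 to ponytail_8 form a chain to MSTR_ponytail, linked by constraints with identical narrow swing and twist limits.
# Unconstrained spine capsules provide torso collision shapes, and spine_05 adds a large offset capsule that appears to stand in for the backpack.
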PhysicsAsset /BRCosmetics/Characters/Player/Female/Medium/Bodies/F_MED_MincePounce/Meshes/Parts/F_MED_MincePounce_FaceAcc_Physics_Ponytail_backpack
bodies {
 MSTR_ponytail [{shapeType=Capsule, length=1, radius=1, shapeName=MSTR_ponytail_capsule, rotation=<rot 0 -90 0>}]
 ponytail_5 [{shapeType=Capsule, length=0.4796, radius=5.8204, shapeName=ponytail_5_capsule, center=<3.0942 0.3602 1.9433>, rotation=<rot 90 0 0>}]
 ponytail_6 [{shapeType=Capsule, length=1.1852, radius=3.8148, shapeName=ponytail_6_capsule, center=<3.2561 0 0.7907>, rotation=<rot 90 0 0>}]
 ponytail_7 [{shapeType=Capsule, length=2.8112, radius=1.4489, shapeName=ponytail_7_capsule, center=<2.8632 0 -0.1467>, rotation=<rot 90 0 0>}]
 ponytail_8 [{shapeType=Capsule, length=3, radius=2, shapeName=ponytail_8_capsule, center=<2.8632 0 0>, rotation=<rot 90 0 0>}]
 spine_01 [{shapeType=Capsule, length=11.382, radius=9.6163, shapeName=spine_01_capsule, center=<0 0.1704 0>, rotation=<rot 0 -90 0>}]
 spine_03 [{shapeType=Capsule, length=6.1673, radius=9.7482, shapeName=spine_03_capsule, center=<0 1.3853 0>, rotation=<rot 0 -90 0>}]
 spine_04 [{shapeType=Capsule, length=8.1307, radius=9.2414, shapeName=spine_04_capsule}]
 spine_05 [
  {shapeType=Capsule, length=5.1833, radius=9.7645, shapeName=spine_05_capsule, center=<3.6573 -0.2952 0>, rotation=<rot 0 -90 0>}
  {shapeType=Capsule, length=11.2105, radius=14.9318, shapeName=spine_05_capsule, center=<1.4312 -27.1032 0>, rotation=<rot 90 -3.4447 0>}
 ]
}
constraints [
 {bone1Name=ponytail_8, bone2Name=ponytail_7, swing1LimitDegrees=15, swing2LimitDegrees=30, twistLimitDegrees=5}
 {bone1Name=ponytail_7, bone2Name=ponytail_6, swing1LimitDegrees=15, swing2LimitDegrees=30, twistLimitDegrees=5}
 {bone1Name=ponytail_6, bone2Name=ponytail_5, swing1LimitDegrees=15, swing2LimitDegrees=30, twistLimitDegrees=5}
 {bone1Name=ponytail_5, bone2Name=MSTR_ponytail, swing1LimitDegrees=15, swing2LimitDegrees=30, twistLimitDegrees=5}
]
