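# Anim Blueprint for the F_MED_MincePounceHelmet face accessory, simulating ponytail physics atop the parent mesh's copied pose.
# A backpack check selects between two Control Rig plus RigidBody branches, then a deform Control Rig finishes the pose.
# The backpack branch simulates only with the MincePounce backpack equipped.
# Each update derives gravity for both simulations, and sim-space settings for the no-backpack one, from crouch or default per-state parameters via a deprecated helper.
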AnimBlueprint /BRCosmetics/Characters/Player/Female/Medium/Bodies/F_MED_MincePounceHelmet/Meshes/Parts/F_MED_MincePounceHelmet_FaceAcc_AnimBP
target_skeleton=&/F_MED_MincePounceHelmet_FaceAcc_Skeleton
graphs [
 AnimGraph {nodes [
   n1 = Control Rig(Source=Copy Pose From Mesh() {bUseAttachedParent=true}) {controlRigAssetReference {blueprintRigClass=/BRCosmetics/Characters/Player/Female/Medium/Bodies/F_MED_MincePounce/Meshes/Parts/F_MED_MincePounce_FaceAcc_CtrlRig.F_MED_MincePounce_FaceAcc_CtrlRig_C}}
   n2 = RigidBody!(
    ComponentPose=Local To Component(LocalPose=@n1)
    Alpha
    OverrideWorldGravity:Vector=Get Gravity Override()
    ExternalForce
    SimSpaceSettings:Sim Space Settings Structure=Get!() -> (
     RBAN Settings Out_ComponentLinearAccScale_28_#1:Vector=~
     RBAN Settings Out_ComponentLinearVelScale_31_#2:Vector=~
     RBAN Settings Out_SimSpaceSettings_34_#3:Sim Space Settings Structure
    ).RBAN Settings Out_SimSpaceSettings_34_#3
   ) -> (Pose) {
    bOverrideWorldGravity=true
    overridePhysicsAsset=/BRCosmetics/Characters/Player/Female/Medium/Bodies/F_MED_MincePounce/Meshes/Parts/F_MED_MincePounce_FaceAcc_Physics_Ponytail
    simulationSpace=BaseBoneSpace
    baseBoneRef {boneName=spine_04}
    componentAppliedLinearAccClamp=<0 0 0>
    simSpaceSettings {maxAngularAcceleration=6, maxAngularVelocity=8, maxLinearAcceleration=100, maxLinearVelocity=200, worldAlpha=1}
   }
   n3 = Control Rig(Source=Copy Pose From Mesh() {bUseAttachedParent=true}) {controlRigAssetReference {blueprintRigClass=/BRCosmetics/Characters/Player/Female/Medium/Bodies/F_MED_MincePounce/Meshes/Parts/F_MED_MincePounce_FaceAcc_CtrlRig_backpack.F_MED_MincePounce_FaceAcc_CtrlRig_backpack_C}}
   n4 = RigidBody!(
    ComponentPose=Local To Component(LocalPose=@n3)
    bAlphaBoolEnabled:bool=Get is_MincePounce_Backpack() -> (is_MincePounce_Backpack:bool)
    OverrideWorldGravity=Get Gravity Override()
    ExternalForce
   ) -> (Pose) {
    alphaInputType=Bool
    bOverrideWorldGravity=true
    overridePhysicsAsset=/BRCosmetics/Characters/Player/Female/Medium/Bodies/F_MED_MincePounce/Meshes/Parts/F_MED_MincePounce_FaceAcc_Physics_Ponytail_backpack
    simulationSpace=BaseBoneSpace
    baseBoneRef {boneName=spine_04}
    componentAppliedLinearAccClamp=<0 0 0>
    simSpaceSettings {maxAngularAcceleration=6, maxAngularVelocity=8, maxLinearAcceleration=100, maxLinearVelocity=200, worldAlpha=1}
   }
   n5 = Control Rig(Source=Blend Poses by bool(
     BlendPose_0=Component To Local(ComponentPose=@n4)
     BlendPose_1=Component To Local(ComponentPose=@n2)
     bActiveValue=Get is_WearingBackpack() -> (is_WearingBackpack:bool)
    )) {controlRigAssetReference {blueprintRigClass=/BRCosmetics/Characters/Player/Female/Medium/Bodies/F_MED_MincePounce/Meshes/Parts/F_MED_MincePounce_FaceAcc_CtrlRig_deform.F_MED_MincePounce_FaceAcc_CtrlRig_deform_C}}
   Output Pose(Result=@n5) {layerGroup=None}
  ]}
 EventGraph {nodes [
   Event BlueprintUpdateAnimation()
   > n1 = Branch(Condition=Get bIsCrouching())
   > n2 = Gravity Override State Function DEPRECATED(
    CharacterPartAnimInstance=Self-Reference()
    GameTimeInSeconds=GetGameTimeInSeconds()
    GravityOverrideParams=Get Gravity Override Params crounch() -> (Gravity Override Params crounch:Gravity Override Params Struct Structure)
    RBAN Params=Get RBAN Params crounch() -> (RBAN Params crounch:RBANParams Struct Structure)
   )
   > Set Gravity Override(Gravity Override=@n2)
   > Set RBAN Settings Out(RBAN Settings Out:RBANSetting Struct Structure=@n2.RBAN SettingsOut) -> (Output_Get:RBANSetting Struct Structure)
   n3 = Gravity Override State Function DEPRECATED(
    execute=@n1.else
    CharacterPartAnimInstance=Self-Reference()
    GameTimeInSeconds=GetGameTimeInSeconds()
    GravityOverrideParams=Get Gravity Override Params()
    RBAN Params=Get RBAN Params()
   )
   > Set Gravity Override(Gravity Override=@n3)
   > Set RBAN Settings Out(RBAN Settings Out=@n3.RBAN SettingsOut)
   n4 = Event BlueprintInitializeAnimation()
   n5 = Event OnCharacterCustomizationLoaded()
   n6 = For Each Loop with Break!(Exec:Exec=[@n4.then, @n5.then], Array:Array of Custom Character Part Object References=@n5.CustomParts, Break:Exec=@n8.then) -> (LoopBody:Exec, Array Element:Custom Character Part Object Reference, Array Index:int, Completed:Exec)
   > n7 = Set is_MincePounce_Backpack(is_MincePounce_Backpack:bool="Equal (Object)"(A=@n6, B=/BRCosmetics/Characters/CharacterParts/Backpacks/CP_Backpack_MincePounce)) -> (Output_Get:bool)
   > n8 = Branch(Condition=@n7)
   n9 = CheckForBackbling()
   Set is_WearingBackpack(execute=@n6.Completed, is_WearingBackpack:bool=@n9) -> (Output_Get:bool)
   > n10 = Branch(Condition=@n9.bIsBackblingACape)
   > Set is_WearingCape(is_WearingCape:bool=true) -> (Output_Get:bool)
   Set is_WearingCape(execute=@n10.else, is_WearingCape=false)
  ]}
]
variables [
 Gravity Override {type=Vector}
 Gravity Override Params {type=GravityOverrideParamsStruct}
 Gravity Override Params crounch {type=GravityOverrideParamsStruct, default {onGround_Moving {gravityOverride=<0 -800 -980>}}}
 is_MincePounce_Backpack {type=boolean}
 is_WearingBackpack {type=boolean}
 is_WearingCape {type=boolean}
 RBAN Params {
  type=RBANParamsStruct
  default {
   dBNO {simSpaceSettings {
     maxAngularAcceleration=6
     maxAngularVelocity=8
     maxLinearAcceleration=100
     maxLinearVelocity=250
     velocityScaleZ=1
     externalLinearDragV=<0 0 0>
    }}
   dBNO_Carried {simSpaceSettings {
     maxAngularAcceleration=6
     maxAngularVelocity=8
     maxLinearAcceleration=100
     maxLinearVelocity=250
     velocityScaleZ=1
     externalLinearDragV=<0 0 0>
    }}
   emoteOrMelee {simSpaceSettings {
     maxAngularAcceleration=6
     maxAngularVelocity=8
     maxLinearAcceleration=100
     maxLinearVelocity=250
     velocityScaleZ=1
     externalLinearDragV=<0 0 0>
    }}
   front_End {simSpaceSettings {
     maxAngularAcceleration=6
     maxAngularVelocity=8
     maxLinearAcceleration=100
     maxLinearVelocity=250
     velocityScaleZ=1
     externalLinearDragV=<0 0 0>
    }}
   onGround_Crouch_Moving {simSpaceSettings {
     maxAngularAcceleration=6
     maxAngularVelocity=8
     maxLinearAcceleration=100
     maxLinearVelocity=250
     velocityScaleZ=1
     externalLinearDragV=<0 0 0>
    }}
   onGround_Falling {simSpaceSettings {
     maxAngularAcceleration=6
     maxAngularVelocity=8
     maxLinearAcceleration=100
     maxLinearVelocity=250
     velocityScaleZ=1
     externalLinearDragV=<0 0 0>
    }}
   onGround_Moving {
    componentLinearAccScale=<0 0 0>
    componentLinearVelScale=<0 0 0>
    simSpaceSettings {
     maxAngularAcceleration=6
     maxAngularVelocity=8
     maxLinearAcceleration=100
     maxLinearVelocity=250
     velocityScaleZ=1
     externalLinearDragV=<0 0 0>
    }
   }
   onGround_Standing {
    componentLinearAccScale=<0 0 0>
    componentLinearVelScale=<0 0 0>
    simSpaceSettings {
     maxAngularAcceleration=6
     maxAngularVelocity=8
     maxLinearAcceleration=100
     maxLinearVelocity=250
     velocityScaleZ=1
     externalLinearDragV=<0 0 0>
    }
   }
   rail_Grinding {simSpaceSettings {
     maxAngularAcceleration=6
     maxAngularVelocity=8
     maxLinearAcceleration=100
     maxLinearVelocity=250
     velocityScaleZ=1
     worldAlpha=1
     externalLinearDragV=<0 0 0>
    }}
   riding_Moving {simSpaceSettings {
     maxAngularAcceleration=1000
     maxAngularVelocity=3
     maxLinearAcceleration=300
     maxLinearVelocity=700
     velocityScaleZ=0
     externalLinearDragV=<0 0.5 0>
    }}
   riding_Moving_Targeting {simSpaceSettings {
     maxAngularAcceleration=6
     maxAngularVelocity=8
     maxLinearAcceleration=100
     maxLinearVelocity=250
     velocityScaleZ=1
     externalLinearDragV=<0 0 0>
    }}
   riding_Standing {simSpaceSettings {
     maxAngularAcceleration=6
     maxAngularVelocity=8
     maxLinearAcceleration=100
     maxLinearVelocity=250
     velocityScaleZ=1
     externalLinearDragV=<0 0 0>
    }}
   skydive_DivingDown {
    componentLinearAccScale=<0 0 0>
    componentLinearVelScale=<0 0 0>
    simSpaceSettings {
     maxAngularAcceleration=6
     maxAngularVelocity=6
     maxLinearAcceleration=100
     maxLinearVelocity=250
     velocityScaleZ=1
     externalLinearDragV=<0 0.5 0>
    }
   }
   skydive_FloatingDown {
    componentLinearAccScale=<0 0.5 0>
    componentLinearVelScale=<0 0.5 0>
    simSpaceSettings {
     maxAngularAcceleration=6
     maxAngularVelocity=8
     maxLinearAcceleration=100
     maxLinearVelocity=250
     velocityScaleZ=1
     externalLinearDragV=<0 0.5 0>
    }
   }
   skydive_GoingUp {simSpaceSettings {
     maxAngularAcceleration=6
     maxAngularVelocity=8
     maxLinearAcceleration=100
     maxLinearVelocity=250
     velocityScaleZ=1
     externalLinearDragV=<0 0.5 0>
    }}
   skydive_Parachuting {
    componentLinearAccScale=<0 0 0>
    componentLinearVelScale=<0 0 0>
    simSpaceSettings {
     maxAngularAcceleration=6
     maxAngularVelocity=8
     maxLinearAcceleration=100
     maxLinearVelocity=250
     velocityScaleZ=1
     externalLinearDragV=<0 0.5 0>
    }
   }
   skydive_Surfing {simSpaceSettings {
     maxAngularAcceleration=6
     maxAngularVelocity=8
     maxLinearAcceleration=100
     maxLinearVelocity=250
     velocityScaleZ=1
     externalLinearDragV=<0 0 0>
    }}
   swimming {simSpaceSettings {
     maxAngularAcceleration=6
     maxAngularVelocity=8
     maxLinearAcceleration=100
     maxLinearVelocity=250
     velocityScaleZ=1
     externalLinearDragV=<0 0 0>
    }}
   tactical_Sprint {
    componentLinearAccScale=<0 0 0>
    simSpaceSettings {
     maxAngularAcceleration=6
     maxAngularVelocity=8
     maxLinearAcceleration=100
     maxLinearVelocity=250
     velocityScaleZ=1
     externalLinearDragV=<0 0 0>
    }
   }
   ziplining {simSpaceSettings {
     maxAngularAcceleration=6
     maxAngularVelocity=8
     maxLinearAcceleration=100
     maxLinearVelocity=250
     velocityScaleZ=1
     externalLinearDragV=<0 0 0>
    }}
  }
 }
 RBAN Params crounch {
  type=RBANParamsStruct
  default {
   dBNO {simSpaceSettings {
     maxAngularAcceleration=6
     maxAngularVelocity=8
     maxLinearAcceleration=100
     maxLinearVelocity=250
     velocityScaleZ=1
     externalLinearDragV=<0 0 0>
    }}
   dBNO_Carried {simSpaceSettings {
     maxAngularAcceleration=6
     maxAngularVelocity=8
     maxLinearAcceleration=100
     maxLinearVelocity=250
     velocityScaleZ=1
     externalLinearDragV=<0 0 0>
    }}
   emoteOrMelee {simSpaceSettings {
     maxAngularAcceleration=6
     maxAngularVelocity=8
     maxLinearAcceleration=100
     maxLinearVelocity=250
     velocityScaleZ=1
     externalLinearDragV=<0 0 0>
    }}
   front_End {simSpaceSettings {
     maxAngularAcceleration=6
     maxAngularVelocity=8
     maxLinearAcceleration=100
     maxLinearVelocity=250
     velocityScaleZ=1
     externalLinearDragV=<0 0 0>
    }}
   onGround_Crouch_Moving {simSpaceSettings {
     maxAngularAcceleration=6
     maxAngularVelocity=8
     maxLinearAcceleration=40
     maxLinearVelocity=40
     velocityScaleZ=1
     externalLinearDragV=<0 0 0>
    }}
   onGround_Falling {simSpaceSettings {
     maxAngularAcceleration=6
     maxAngularVelocity=8
     maxLinearAcceleration=100
     maxLinearVelocity=250
     velocityScaleZ=1
     externalLinearDragV=<0 0 0>
    }}
   onGround_Moving {
    componentLinearAccScale=<0 0 0>
    componentLinearVelScale=<0 0 0>
    simSpaceSettings {
     maxAngularAcceleration=6
     maxAngularVelocity=8
     maxLinearAcceleration=40
     maxLinearVelocity=40
     velocityScaleZ=1
     externalLinearDragV=<0 0 0>
    }
   }
   onGround_Standing {
    componentLinearAccScale=<0 0 0>
    componentLinearVelScale=<0 0 0>
    simSpaceSettings {
     maxAngularAcceleration=6
     maxAngularVelocity=8
     maxLinearAcceleration=100
     maxLinearVelocity=250
     velocityScaleZ=1
     externalLinearDragV=<0 0 0>
    }
   }
   rail_Grinding {simSpaceSettings {
     maxAngularAcceleration=6
     maxAngularVelocity=8
     maxLinearAcceleration=100
     maxLinearVelocity=250
     velocityScaleZ=1
     worldAlpha=1
     externalLinearDragV=<0 0 0>
    }}
   riding_Moving {simSpaceSettings {
     maxAngularAcceleration=1000
     maxAngularVelocity=3
     maxLinearAcceleration=300
     maxLinearVelocity=700
     velocityScaleZ=0
     externalLinearDragV=<0 0.5 0>
    }}
   riding_Moving_Targeting {simSpaceSettings {
     maxAngularAcceleration=6
     maxAngularVelocity=8
     maxLinearAcceleration=100
     maxLinearVelocity=250
     velocityScaleZ=1
     externalLinearDragV=<0 0 0>
    }}
   riding_Standing {simSpaceSettings {
     maxAngularAcceleration=6
     maxAngularVelocity=8
     maxLinearAcceleration=100
     maxLinearVelocity=250
     velocityScaleZ=1
     externalLinearDragV=<0 0 0>
    }}
   skydive_DivingDown {
    componentLinearAccScale=<0 0 0>
    componentLinearVelScale=<0 0 0>
    simSpaceSettings {
     maxAngularAcceleration=6
     maxAngularVelocity=6
     maxLinearAcceleration=100
     maxLinearVelocity=250
     velocityScaleZ=1
     externalLinearDragV=<0 0.5 0>
    }
   }
   skydive_FloatingDown {
    componentLinearAccScale=<0 0.5 0>
    componentLinearVelScale=<0 0.5 0>
    simSpaceSettings {
     maxAngularAcceleration=6
     maxAngularVelocity=8
     maxLinearAcceleration=100
     maxLinearVelocity=250
     velocityScaleZ=1
     externalLinearDragV=<0 0 0>
    }
   }
   skydive_GoingUp {simSpaceSettings {
     maxAngularAcceleration=6
     maxAngularVelocity=8
     maxLinearAcceleration=100
     maxLinearVelocity=250
     velocityScaleZ=1
     externalLinearDragV=<0 0.5 0>
    }}
   skydive_Parachuting {
    componentLinearAccScale=<0 0 0>
    componentLinearVelScale=<0 0 0>
    simSpaceSettings {
     maxAngularAcceleration=6
     maxAngularVelocity=8
     maxLinearAcceleration=100
     maxLinearVelocity=250
     velocityScaleZ=1
     externalLinearDragV=<0 0.5 0>
    }
   }
   skydive_Surfing {simSpaceSettings {
     maxAngularAcceleration=6
     maxAngularVelocity=8
     maxLinearAcceleration=100
     maxLinearVelocity=250
     velocityScaleZ=1
     externalLinearDragV=<0 0 0>
    }}
   swimming {simSpaceSettings {
     maxAngularAcceleration=6
     maxAngularVelocity=8
     maxLinearAcceleration=100
     maxLinearVelocity=250
     velocityScaleZ=1
     externalLinearDragV=<0 0 0>
    }}
   tactical_Sprint {
    componentLinearAccScale=<0 0 0>
    simSpaceSettings {
     maxAngularAcceleration=6
     maxAngularVelocity=8
     maxLinearAcceleration=100
     maxLinearVelocity=250
     velocityScaleZ=1
     externalLinearDragV=<0 0 0>
    }
   }
   ziplining {simSpaceSettings {
     maxAngularAcceleration=6
     maxAngularVelocity=8
     maxLinearAcceleration=100
     maxLinearVelocity=250
     velocityScaleZ=1
     externalLinearDragV=<0 0 0>
    }}
  }
 }
 RBAN Settings Out {
  type=RBANSettingStruct
  default {
   componentLinearAccScale=<0 0 0>
   componentLinearVelScale=<0 0 0>
   simSpaceSettings {
    dampingAlpha=1
    maxAngularAcceleration=1000
    maxAngularVelocity=3
    maxLinearAcceleration=300
    maxLinearVelocity=700
    velocityScaleZ=0
    worldAlpha=1
    externalAngularVelocity=<0 0 0>
    externalLinearDragV=<0 0.5 0>
    externalLinearVelocity=<0 0 0>
   }
  }
 }
]
